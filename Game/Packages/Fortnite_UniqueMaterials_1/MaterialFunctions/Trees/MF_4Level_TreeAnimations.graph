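# Material function that adds hierarchical wind animation to tree materials, for up to four Pivot Painter 2 levels.
# It walks parent indices in the data texture, selecting each level's pivot by hierarchy depth.
# Four Fortnite_FoliageLevelAnimation passes chain rotated normals and sum world position offsets, with a storm mask blending levels 2 and 3 toward storm settings.
# Low and Medium quality skip the motion; a debug switch shows depth as emissive color.
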MaterialFunction /Game/Packages/Fortnite_UniqueMaterials_1/MaterialFunctions/Trees/MF_4Level_TreeAnimations
graph {nodes [
  n1 = FunctionInput() {inputType=FunctionInput_MaterialAttributes}
  n2 = TextureObjectParameter() {
   group=Vertex Shader Textures
   parameterName=Position and Index Texture
   samplerType=SAMPLERTYPE_LinearColor
   texture=/Engine/Functions/Engine_MaterialFunctions02/PivotPainter2/Black_1x1_EXR_Texture
  }
  n3 = TextureProperty(Tex=@n2)
  n4 = MaterialFunctionCall!(Data Texture Dimensions=@n3, Pivot Painter UV Coordinates=TextureCoordinate() {coordinateIndex=1}) {materialFunction=/Engine/Functions/Engine_MaterialFunctions02/PivotPainter2/ms_PivotPainter2_CalculateMeshElementIndex}
  n5 = MaterialFunctionCall!("Current Index (ParentUVs)"=@n4.Index, UVs=TextureCoordinate() {coordinateIndex=1}, "Texture resolution (ParentUVs)"=@n3) {materialFunction=/Engine/Functions/Engine_MaterialFunctions02/PivotPainter2/ms_PivotPainter2_SampleLayerData}
  n6 = MaterialFunctionCall("Current Index (ParentUVs)"=@n5.Parent Index, UVs=@n5.Parent Index, "Texture resolution (ParentUVs)"=@n3) {materialFunction=&ms_PivotPainter2_SampleLayerData}
  n7 = MaterialFunctionCall("Current Index (ParentUVs)"=@n6.Parent Index, UVs=@n6.Parent Index, "Texture resolution (ParentUVs)"=@n3) {materialFunction=&ms_PivotPainter2_SampleLayerData}
  n8 = MaterialFunctionCall("Current Index (ParentUVs)"=@n7.Parent Index, UVs=@n7.Parent Index, "Texture resolution (ParentUVs)"=@n3) {materialFunction=&ms_PivotPainter2_SampleLayerData}
  n9 = Ceil(None=Add(A=Add(A=@n6.Is Child?, B=@n5.Is Child?), B=Add(A=@n7.Is Child?, B=@n8.Is Child?)))
  n10 = Clamp(None=Subtract(A=@n9, B=Constant() {r=2}))
  n11 = LinearInterpolate(A=@n5.Position and Parent Index, B=@n6.Position and Parent Index, Alpha=@n10)
  n12 = Constant() {r=1}
  n13 = Clamp(None=Divide(A=Subtract(A=@n9, B=@n12)))
  n14 = MaterialFunctionCall!(1=@n5.Position and Parent Index, 2=@n6.Position and Parent Index, 3=@n7.Position and Parent Index, 4=@n8.Position and Parent Index, A=@n13) {desc=position, materialFunction=/Engine/Functions/Engine_MaterialFunctions02/Lerp_Multiple_Float4}
  n15 = Divide(A=@n9) {constB=3}
  n16 = MaterialFunctionCall!(Material=@n1) {materialFunction=/Engine/Functions/MaterialLayerFunctions/MatLayerBlend_BreakNormal}
  n17 = MaterialFunctionCall!(
   Enable=StaticBoolParameter() {group=Wind Setting 1, parameterName=Animate Level 1}
   Enable Normal Rotation=StaticBoolParameter() {group=Wind Setting 1, parameterName=Animate Level 1 Normals}
   Pivot Position=MaterialFunctionCall(1=@n5.Position and Parent Index, 2=@n6.Position and Parent Index, 3=@n7.Position and Parent Index, 4=@n8.Position and Parent Index, A=@n15) {desc=position, materialFunction=&Lerp_Multiple_Float4}.Lerp 4 Inputs
   X-Axis Texture And X Extent Alpha=MaterialFunctionCall(
    1=@n5.X axis vector and X extent
    2=@n6.X axis vector and X extent
    3=@n7.X axis vector and X extent
    4=@n8.X axis vector and X extent
    A=@n15
   ) {desc=Axis, materialFunction=&Lerp_Multiple_Float4}.Lerp 4 Inputs
   Optional - world space pixel normal=@n16.Normal
   Motion Dampening Falloff Radius=ScalarParameter() {defaultValue=1, group=Wind Setting 1, parameterName=Radius Multiplier 1}
   Wind Gust Angle Rotation=ScalarParameter() {defaultValue=0.0092, group=Wind Setting 1, parameterName=Max Rotation 1, sliderMax=10}
   Wind Gust Offset=ScalarParameter() {group=Wind Setting 1, parameterName=Wind offset 1}
   TimeRate_CodeReferenced=ScalarParameter() {defaultValue=0.1, group=Wind Setting 1, parameterName=Wind_1_Speed_CodeReferenced}
   TimeOffset_CodeControlled=ScalarParameter() {group=Code Controlled, parameterName=Wind_1_Offset_CodeReferenced}
   Make Motion Falloff Element Size Specific
   Use Pivots To Sample Wind Intensity At Higher Level Branches
   Pivot of point 1 level higher in the hierarchy=@n14.Lerp 3 Inputs
   UseWrapAroundLogic=StaticBoolParameter() {group=Wind Setting 1, parameterName=Wind 1 - Use Mesh WrapAroundPivot Method For Animation}
   RigidAnimation?
   Advanced - Parent Rotation Angle Influence=Constant()
   Random Rotation Texture Sample Scale=ScalarParameter() {defaultValue=250, group=Wind Setting 1, parameterName=Wind 1 Random Rotation Change Rate - Do Not Animate}
   Random Rotation Influence=ScalarParameter() {defaultValue=0.25, group=Wind Setting 1, parameterName=Wind 1 Random Rotation Influence}
   Wind Shelter Bias=ScalarParameter() {defaultValue=-1, group=Wind Setting 1, parameterName=Wind 1 Shelter Bias}
   Wind Shelter Scale=ScalarParameter() {defaultValue=-2, group=Wind Setting 1, parameterName=Wind 1 Shelter Scale}
   Optional - Wind Turbulence and Gust Mag
   Model Depth=@n9
   rotation Depth=Constant()
   Offset AnimPhase Via X Axis Direction=StaticBoolParameter() {group=Wind Setting 1, parameterName="Wind 1 - Use Vertex Position to offset the Animation - X Axis (Willow Tree)"}
   Vertex Phase Offset X Axis=ScalarParameter() {defaultValue=-2000, group=Wind Setting 1, parameterName=Wind 1 - Vertex Phase Offset X Axis}
   "Use \"Wind Shelter\" Directional Masks"=StaticBoolParameter() {group=Wind Setting 1, parameterName=Wind 1 - Use The 'Wind Shelter' Directional Mask}
   Control Harsh Falloff Around Root=StaticBoolParameter() {group=Wind Setting 2, parameterName=Wind 1 - Use Harsh Motion Root Mask}
   Percentage of Branch Length For Harsh Falloff=ScalarParameter() {defaultValue=0.01, group=Wind Setting 2, parameterName=Wind 1 - Harsh Root-Based Branch Length Animation Mask}
   Wind Direction=VectorParameter() {group=Code Controlled, parameterName=WindVector_CodeControlled_3, defaultValue=<rgba 1 0 0 1>}.RGB
  ) {materialFunction=/Game/Packages/Fortnite_UniqueMaterials_1/MaterialFunctions/Fortnite_FoliageLevelAnimation}
  n18 = MaterialFunctionCall() {materialFunction=/Game/Athena/Environments/Landscape/MaterialFunctions/MF_StormMask}
  n19 = MaterialFunctionCall(
   Enable=StaticBoolParameter() {group=Wind Setting 2, parameterName=Animate Level 2}
   Enable Normal Rotation=StaticBoolParameter() {group=Wind Setting 2, parameterName=Animate Level 2 Normals}
   Pivot Position=@n14.Lerp 3 Inputs
   X-Axis Texture And X Extent Alpha=MaterialFunctionCall(
    1=@n5.X axis vector and X extent
    2=@n6.X axis vector and X extent
    3=@n7.X axis vector and X extent
    4=@n8.X axis vector and X extent
    A=@n13
   ) {desc=Axis, materialFunction=&Lerp_Multiple_Float4}.Lerp 3 Inputs
   Optional - world space pixel normal=@n17.Rotated World Space Pixel Normal
   Motion Dampening Falloff Radius=LinearInterpolate(
    A=ScalarParameter() {defaultValue=1, group=Wind Setting 2, parameterName=Radius Multiplier 2}
    B=ScalarParameter() {defaultValue=0.1, group=Wind Setting 2, parameterName=Radius Multiplier STORM}
    Alpha=@n18.Result
   )
   Wind Gust Angle Rotation=LinearInterpolate(
    A=ScalarParameter() {defaultValue=0.0092, group=Wind Setting 2, parameterName=Max Rotation 2, sliderMax=10}
    B=ScalarParameter() {defaultValue=0.04, group=Wind Setting 2, parameterName=Max Rotation STORM, sliderMax=10}
    Alpha=@n18.Result
   )
   Wind Gust Offset=ScalarParameter() {group=Wind Setting 2, parameterName=Wind offset 2}
   TimeRate_CodeReferenced=ScalarParameter() {defaultValue=0.1, group=Wind Setting 2, parameterName=Wind_2_Speed_CodeReferenced}
   TimeOffset_CodeControlled=ScalarParameter() {group=Code Controlled, parameterName=Wind_2_Offset_CodeReferenced}
   Pivot of point 1 level higher in the hierarchy=@n11
   UseWrapAroundLogic=StaticBoolParameter() {group=Wind Setting 2, parameterName=Wind 2 - Use Mesh WrapAroundPivot Method For Animation}
   Advanced - Parent Rotation Angle Influence=Constant()
   Random Rotation Texture Sample Scale=ScalarParameter() {defaultValue=250, group=Wind Setting 2, parameterName=Wind 2 Random Rotation Change Rate - Do Not Animate}
   Random Rotation Influence=ScalarParameter() {defaultValue=0.25, group=Wind Setting 2, parameterName=Wind 2 Random Rotation Influence, sliderMax=2}
   Wind Shelter Bias=ScalarParameter() {defaultValue=-1, group=Wind Setting 2, parameterName=Wind 2 Shelter Bias}
   Wind Shelter Scale=ScalarParameter() {defaultValue=-2, group=Wind Setting 2, parameterName=Wind 2 Shelter Scale}
   Model Depth=@n9
   rotation Depth=Constant() {r=1}
   Offset AnimPhase Via X Axis Direction=StaticBoolParameter() {group=Wind Setting 2, parameterName="Wind 2 - Use Vertex Position to offset the Animation - X Axis (Willow Tree)"}
   Vertex Phase Offset X Axis=ScalarParameter() {defaultValue=-2000, group=Wind Setting 2, parameterName=Wind 2 - Vertex Phase Offset X Axis}
   "Use \"Wind Shelter\" Directional Masks"=StaticBoolParameter() {defaultValue=true, group=Wind Setting 2, parameterName=Wind 2 - Use The 'Wind Shelter' Directional Mask}
   Control Harsh Falloff Around Root=StaticBoolParameter() {group=Wind Setting 2, parameterName=Wind 2 - Use Harsh Motion Root Mask}
   Percentage of Branch Length For Harsh Falloff=ScalarParameter() {defaultValue=0.01, group=Wind Setting 2, parameterName=Wind 2 - Harsh Root-Based Branch Length Animation Mask}
   Wind Direction=VectorParameter() {group=Code Controlled, parameterName=WindVector_CodeControlled_2, defaultValue=<rgba 1 0 0 1>}.RGB
  ) {materialFunction=&Fortnite_FoliageLevelAnimation}
  n20 = MaterialFunctionCall(
   Enable=StaticBoolParameter() {group=Wind Setting 3, parameterName=Animate Level 3}
   Enable Normal Rotation=StaticBoolParameter() {group=Wind Setting 3, parameterName=Animate Level 3 Normals}
   Pivot Position=@n11
   X-Axis Texture And X Extent Alpha=LinearInterpolate(A=@n5.X axis vector and X extent, B=@n6.X axis vector and X extent, Alpha=@n10)
   Optional - world space pixel normal=@n19.Rotated World Space Pixel Normal
   Motion Dampening Falloff Radius=LinearInterpolate(
    A=ScalarParameter() {defaultValue=1, group=Wind Setting 3, parameterName=Radius Multiplier 3}
    B=ScalarParameter() {defaultValue=1, group=Wind Setting 3, parameterName=Radius Multiplier 3 STORM}
    Alpha=@n18.Result
   )
   Wind Gust Angle Rotation=LinearInterpolate(
    A=ScalarParameter() {defaultValue=0.0092, group=Wind Setting 3, parameterName=Max Rotation 3, sliderMax=10}
    B=ScalarParameter() {defaultValue=0.1, group=Wind Setting 3, parameterName=Max Rotation 3 STORM, sliderMax=10}
    Alpha=@n18.Result
   )
   Wind Gust Offset=ScalarParameter() {group=Wind Setting 3, parameterName=Wind offset 3}
   TimeRate_CodeReferenced=ScalarParameter() {defaultValue=0.1, group=Wind Setting 3, parameterName=Wind_3_Speed_CodeReferenced}
   TimeOffset_CodeControlled=ScalarParameter() {group=Code Controlled, parameterName=Wind_3_Offset_CodeReferenced}
   Make Motion Falloff Element Size Specific=StaticBoolParameter() {defaultValue=true, group=Wind Setting 3, parameterName=Wind 3 Make Falloff Element Size Specific}
   Pivot of point 1 level higher in the hierarchy=@n5.Position and Parent Index
   UseWrapAroundLogic=StaticBoolParameter() {group=Wind Setting 3, parameterName=Wind 3 - Use Mesh WrapAroundPivot Method For Animation}
   Advanced - Parent Rotation Angle Influence=Constant()
   Random Rotation Texture Sample Scale=ScalarParameter() {defaultValue=250, group=Wind Setting 3, parameterName=Wind 3 Random Rotation Change Rate - Do Not Animate}
   Random Rotation Influence=ScalarParameter() {defaultValue=0.25, group=Wind Setting 3, parameterName=Wind 3 Random Rotation Influence, sliderMax=2}
   Wind Shelter Bias=ScalarParameter() {defaultValue=-1, group=Wind Setting 3, parameterName=Wind 3 Shelter Bias}
   Wind Shelter Scale=ScalarParameter() {defaultValue=-2, group=Wind Setting 3, parameterName=Wind 3 Shelter Scale}
   Model Depth=@n9
   rotation Depth=Constant() {r=2}
   Offset AnimPhase Via X Axis Direction=StaticBoolParameter() {group=Wind Setting 3, parameterName="Wind 3 - Use Vertex Position to offset the Animation - X Axis (Willow Tree)"}
   Vertex Phase Offset X Axis=ScalarParameter() {defaultValue=-2000, group=Wind Setting 3, parameterName=Wind 3 - Vertex Phase Offset X Axis}
   "Use \"Wind Shelter\" Directional Masks"=StaticBoolParameter() {defaultValue=true, group=Wind Setting 3, parameterName=Wind 3 - Use The 'Wind Shelter' Directional Mask}
   Control Harsh Falloff Around Root=StaticBoolParameter() {group=Wind Setting 3, parameterName=Wind 3 - Use Harsh Motion Root Mask}
   Percentage of Branch Length For Harsh Falloff=ScalarParameter() {defaultValue=0.01, group=Wind Setting 3, parameterName=Wind 3 - Harsh Root-Based Branch Length Animation Mask}
   Wind Direction=VectorParameter() {group=Code Controlled, parameterName=WindVector_CodeControlled, defaultValue=<rgba 1 0 0 1>}.RGB
  ) {materialFunction=&Fortnite_FoliageLevelAnimation}
  n21 = MaterialFunctionCall(
   Enable=StaticBoolParameter() {group=Wind Setting 4, parameterName=Animate Level 4}
   Enable Normal Rotation=StaticBoolParameter() {group=Wind Setting 4, parameterName=Animate Level 4 Normals}
   Pivot Position=@n5.Position and Parent Index
   X-Axis Texture And X Extent Alpha=@n5.Position and Parent Index
   Optional - world space pixel normal=@n20.Rotated World Space Pixel Normal
   Motion Dampening Falloff Radius=ScalarParameter() {group=Wind Setting 4, parameterName=Radius Multiplier 4}
   Wind Gust Angle Rotation=ScalarParameter() {defaultValue=0.0092, group=Wind Setting 4, parameterName=Max Rotation 4, sliderMax=10}
   Wind Gust Offset=ScalarParameter() {group=Wind Setting 4, parameterName=Wind offset 4}
   TimeRate_CodeReferenced=ScalarParameter() {defaultValue=0.1, group=Wind Setting 4, parameterName=Wind_4_Speed_CodeReferenced}
   TimeOffset_CodeControlled=ScalarParameter() {group=Code Controlled, parameterName=Wind_4_Offset_CodeReferenced}
   Pivot of point 1 level higher in the hierarchy=@n5.Position and Parent Index
   UseWrapAroundLogic=StaticBoolParameter() {group=Wind Setting 4, parameterName=Wind 4 - Use Mesh WrapAroundPivot Method For Animation}
   Advanced - Parent Rotation Angle Influence=Multiply(A=@n19.Advanced - Rotation Angle Animation, B=ScalarParameter() {group=Wind Setting 4, parameterName=Wind 4 - Parent Influence On Leaves})
   Random Rotation Texture Sample Scale=ScalarParameter() {defaultValue=250, group=Wind Setting 4, parameterName=Wind 4 Random Rotation Change Rate - Do Not Animate}
   Random Rotation Influence=ScalarParameter() {defaultValue=0.25, group=Wind Setting 4, parameterName=Wind 4 Random Rotation Influence, sliderMax=2}
   Wind Shelter Bias=ScalarParameter() {defaultValue=-1, group=Wind Setting 4, parameterName=Wind 4 Shelter Bias}
   Wind Shelter Scale=ScalarParameter() {defaultValue=-2, group=Wind Setting 4, parameterName=Wind 4 Shelter Scale}
   Model Depth=@n9
   rotation Depth=Constant() {r=3}
   Offset AnimPhase Via X Axis Direction=StaticBoolParameter() {group=Wind Setting 4, parameterName="Wind 4 - Use Vertex Position to offset the Animation - X Axis (Willow Tree)"}
   Vertex Phase Offset X Axis=ScalarParameter() {defaultValue=-2000, group=Wind Setting 4, parameterName=Wind 4 - Vertex Phase Offset X Axis}
   "Use \"Wind Shelter\" Directional Masks"=StaticBoolParameter() {defaultValue=true, group=Wind Setting 4, parameterName=Wind 4 - Use The 'Wind Shelter' Directional Mask}
   Control Harsh Falloff Around Root=StaticBoolParameter() {group=Wind Setting 4, parameterName=Wind 4 - Use Harsh Motion Root Mask}
   Percentage of Branch Length For Harsh Falloff=ScalarParameter() {defaultValue=0.01, group=Wind Setting 4, parameterName=Wind 4 - Harsh Root-Based Branch Length Animation Mask}
   Wind Direction=VectorParameter() {group=Code Controlled, parameterName=WindVector_CodeControlled_1, defaultValue=<rgba 1 0 0 1>}.RGB
  ) {materialFunction=&Fortnite_FoliageLevelAnimation}
  n22 = Add(A=Add(A=Add(A=@n17.Add to WPO, B=@n19.Add to WPO), B=@n20.Add to WPO), B=@n21.Add to WPO)
  n23 = Constant()
  n24 = MaterialFunctionCall!(
   Material=MaterialFunctionCall!(
    Material=@n1
    Normal=QualitySwitch(Default=@n21.Rotated World Space Pixel Normal, Low=@n16.Normal, High=@n21.Rotated World Space Pixel Normal, Medium=@n16.Normal)
   ) {materialFunction=/Engine/Functions/MaterialLayerFunctions/MatLayerBlend_ReplaceNormals}.Result
   WorldPositionOffset=QualitySwitch(Default=@n22, Low=@n23, High=@n22, Medium=@n23)
  ) {materialFunction=/Engine/Functions/MaterialLayerFunctions/MatLayerBlend_AddWorldPositionOffset}
  n25 = ScalarParameter() {defaultValue=1, parameterName=1.0}
  n26 = MaterialFunctionCall(
   1=MaterialFunctionCall(X=@n25) {materialFunction=&MakeFloat3}.Result
   2=MaterialFunctionCall(Y=@n25) {materialFunction=&MakeFloat3}.Result
   3=MaterialFunctionCall(Z=@n25) {materialFunction=&MakeFloat3}.Result
   4=MaterialFunctionCall(X=@n25, Y=@n25, Z=@n25) {materialFunction=&MakeFloat3}.Result
   A=Divide(A=@n9) {constB=3}
  ) {materialFunction=&Lerp_Multiple_Float3}
  FunctionOutput(None=StaticSwitchParameter(
    True=MaterialFunctionCall!(Material=@n24.Blended Material, Emissive=@n26.Lerp 4 Inputs) {materialFunction=/Engine/Functions/MaterialLayerFunctions/MatLayerBlend_Emissive}.Blended Material
    False=@n24.Blended Material
   ) {group=Debug, parameterName=Debug Hierarchy Depth})
  FunctionOutput(None=@n5.Position and Parent Index) {outputName=Position and Parent Index, sortPriority=3}
  FunctionOutput(None=@n4.Index) {outputName=Mesh Element Index, sortPriority=6}
 ]}
