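# Material function that adds snow to tree materials by tinting the Base Color and Subsurface Color of incoming material attributes.
# The snow mask compares the normal, optionally transformed from tangent space and flipped on back faces, with up, then applies contrast and bias.
# An optional two-sided blend gives back faces an inverted, separately biased mask.
# Static parameters toggle the effect and multiplication by the Helios snow DetailMask.
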MaterialFunction /Game/Packages/Fortnite_UniqueMaterials_1/MaterialFunctions/Trees/MF_Tree_Snow
graph {nodes [
  Constant() {desc=Snow albedo, r=0.7}
  n1 = NamedRerouteUsage()
  n2 = TwoSidedSign()
  n3 = NamedRerouteUsage()
  n4 = StaticSwitch(
   True=Transform(None=@n3)
   False=@n3
   Value=FunctionInput(Preview=StaticBool()) {bUsePreviewValueAsDefault=true, inputName=IsTangentSpaceInput, inputType=FunctionInput_StaticBool, sortPriority=4}
  )
  n5 = StaticBool()
  n6 = Normalize(VectorInput=StaticSwitch(
    True=Multiply(A=@n2, B=@n4)
    False=@n4
    Value=FunctionInput(Preview=@n5) {bUsePreviewValueAsDefault=true, inputName=HasBackFlipFaceNormals, inputType=FunctionInput_StaticBool, sortPriority=50}
   ))
  n7 = Constant3Vector() {constant=<rgba 0 0 1 0>}
  n8 = MaterialFunctionCall(In=Clamp(None=DotProduct(A=@n6, B=@n7)), Contrast=Constant()) {materialFunction=&CheapContrast}
  n9 = Add(A=ScalarParameter() {defaultValue=-0.5, group=Snow, parameterName=SnowNormal_BackFace_Bias}, B=OneMinus(None=@n8.Result))
  n10 = Add(A=@n8.Result, B=ScalarParameter() {defaultValue=-0.2, group=Snow, parameterName=SnowNormal_Bias})
  n11 = Saturate(None=@n2)
  n12 = StaticBool()
  n13 = Saturate(None=StaticSwitch(
    True=LinearInterpolate(A=@n9, B=@n10, Alpha=@n11)
    False=@n10
    Value=FunctionInput(Preview=@n12) {bUsePreviewValueAsDefault=true, inputName=FlipTwoSidedSign, inputType=FunctionInput_StaticBool, sortPriority=4}
   ))
  n14 = StaticSwitchParameter(
   True=Multiply(
    A=MaterialFunctionCall!(
     ScalarWeightMask
     UseSnowOcclusionTextureMask
     ------
     UseMaterialInput
     AdditionalMaskToRemoveSnow
     OutputTangentNormals
     UseAdditionalMaskToRemoveSnow
     UseSnowNormals
     OuputVertexNormalsAsTheSnowNormals
     UseSparkle
     LQSpark
     AllowSnowMicroNormals
     CliffMaterialOnly
     Material
     CliffLandscapeLayerMask
     IsLandscape
     AreInputNormalsTangentSpace
     UseInputNormalsForFakeSnowShadow
    ) {materialFunction=/Game/Athena/Helios/Environments/Landscape/Materials/QualityLevelLayers/MF_Helios_Snow_Default}.DetailMask
    B=@n13
   )
   False=@n13
  ) {group=Snow, parameterName=Use_BRSnow_Worldmask}
  n15 = LinearInterpolate(A=NamedRerouteUsage(), B=VectorParameter() {group=Snow, parameterName=Snow_Color, defaultValue=<rgba 0.599 0.6328 0.7188 1>}.RGB, Alpha=@n14) {constB=0.7}
  n16 = LinearInterpolate(
   A=NamedRerouteUsage()
   B=VectorParameter() {group=Snow, parameterName=Snow_SSS_Color, defaultValue=<rgba 0.3582 0.4617 0.724 1>}.RGB
   Alpha=@n14
  ) {constB=0.7}
  FunctionOutput(None=StaticSwitchParameter(
    True=SetMaterialAttributes!(MaterialAttributes=@n1, BaseColor=@n15, SubsurfaceColor=@n16) {attributeSetTypes [#1, #2]}
    False=NamedRerouteUsage()
   ) {group=Snow, parameterName=Use_BRSnow})
  n17 = GetMaterialAttributes(None=FunctionInput() {inputType=FunctionInput_MaterialAttributes}) {attributeGetTypes [#1, #3, #2]}
  NamedRerouteDeclaration(None=@n17.BaseColor) {name=Basecolor_Input, nodeColor=<rgba 0.1953 0 1 1>}
  NamedRerouteDeclaration(None=@n17.Normal) {name=Normal_Input, nodeColor=<rgba 0.1953 0 1 1>}
  NamedRerouteDeclaration(None=@n17.MaterialAttributes) {name=Material_Input, nodeColor=<rgba 0.1953 0 1 1>}
  NamedRerouteDeclaration(None=@n17.SubsurfaceColor) {name=SSColor_Input, nodeColor=<rgba 0.1953 0 1 1>}
  ScalarParameter() {defaultValue=0.75, desc=Snow SSSColor, group=Snow, parameterName=Snow_SSS_Brightness}
 ]}
