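# Legacy material function that adds Season 7 landscape-style snow to tree materials, outputting two snowed variants plus the snow mask.
# The holiday snowfall function builds the mask from normals, instance or world position and an optional extra mask.
# FullyBlendedMA blends in full snow with distance-faded topo cavity tint, or flat-color snow; the other output only recolors.
# Static switches toggle snow and a color-test desaturation with darkened subsurface, while original WPO is preserved.
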MaterialFunction /Game/Packages/Fortnite_UniqueMaterials_1/MaterialFunctions/Trees/MF_Tree_Snow_Legacy
graph {nodes [
  ActorPositionWS()
  Constant3Vector()
  Constant()
  FunctionInput(Preview=StaticBool()) {bUsePreviewValueAsDefault=true, inputName=UseColorTest, inputType=FunctionInput_StaticBool, previewValue=<1 0 0 1>}
  n1 = FunctionInput() {inputType=FunctionInput_MaterialAttributes}
  n2 = MaterialFunctionCall() {materialFunction=/Game/Athena/Environments/Landscape/MaterialFunctions/Snow/MF_Athena_Snow_01_S7}
  n3 = GetMaterialAttributes(None=@n2.Material) {attributeGetTypes [#1]}
  n4 = Multiply(
   A=@n3.BaseColor
   B=ComponentMask(None=CollectionParameter() {collection=/Game/Athena/Environments/Landscape/MPC/MPC_Landscape_S07, parameterName=Snow_CavTint}) {b=true, g=true, r=true}
  )
  n5 = TextureSampleParameter2D(UVs=MaterialFunctionCall!(WorldPosition) {materialFunction=/Game/Athena/Environments/Landscape/MaterialFunctions/Utilities/MF_Athena_Terrain_Coords}.Result) {
   group=Snow
   parameterName=TopoMask
   samplerType=SAMPLERTYPE_LinearColor
   texture=/Game/Athena/Environments/Landscape/Textures/T_Athena_Terrain_Topo_Mask
  }
  n6 = Multiply(
   A=@n5.B
   B=MaterialFunctionCall!(Distance=ScalarParameter() {defaultValue=25000, parameterName=SnowCavDistance}) {materialFunction=/Game/Athena/Environments/Landscape/MaterialFunctions/MF_TerrainDistanceFade}.Result
  )
  n7 = StaticSwitchParameter(
   True=SetMaterialAttributes!(MaterialAttributes=@n3.MaterialAttributes, BaseColor=LinearInterpolate(A=@n3.BaseColor, B=@n4, Alpha=@n6)) {attributeSetTypes [#1]}
   False=MakeMaterialAttributes!(
    BaseColor=VectorParameter() {parameterName=SnowColor, defaultValue=<rgba 0.4382 0.5161 0.574 1>}.RGB
    Metallic
    Specular=ScalarParameter() {defaultValue=0.42, group=Snow, parameterName=SnowSpec}
    Roughness=ScalarParameter() {defaultValue=0.75, group=Snow, parameterName=SnowRough}
    Anisotropy
    EmissiveColor
    Opacity
    OpacityMask
    Normal=GetMaterialAttributes(None=@n1) {attributeGetTypes [#2]}.Normal
    Tangent
    WorldPositionOffset
    SubsurfaceColor
    ClearCoat
    ClearCoatRoughness
    AmbientOcclusion
    Refraction
    CustomizedUVs_0
    CustomizedUVs_1
    CustomizedUVs_2
    CustomizedUVs_3
    CustomizedUVs_4
    CustomizedUVs_5
    CustomizedUVs_6
    CustomizedUVs_7
    PixelDepthOffset
    ShadingModel
    Displacement
   )
  ) {group=Snow, parameterName=UseFullSnowMaterial}
  n8 = FunctionInput() {bUsePreviewValueAsDefault=true, inputName=ExtraSnowMask, inputType=FunctionInput_Scalar, previewValue=<1 0 0 1>}
  n9 = GetMaterialAttributes(None=@n1) {attributeGetTypes [
    #1
    #2
    #3
    #4
   ]}
  n10 = StaticSwitch(
   True=Transform(None=@n9.Normal)
   False=@n9.Normal
   Value=FunctionInput(Preview=StaticBool()) {bUsePreviewValueAsDefault=true, inputName=ForceUseInWSNormal, inputType=FunctionInput_StaticBool}
  )
  n11 = Constant3Vector()
  n12 = FunctionInput(Preview=StaticBoolParameter() {defaultValue=true, parameterName=UseActorPositionForSnow}) {bUsePreviewValueAsDefault=true, inputName=UseActorPosition, inputType=FunctionInput_StaticBool, previewValue=<1 0 0 1>}
  n13 = MaterialFunctionCall!(
   UseHeightLerp
   UV_WorldMask
   Use_ExtraSnowMask=StaticBoolParameter() {group=Snow, parameterName=UseExtraSnowMask}
   LandscapeBlend_Height
   Landscape_RockMask
   Snow_Mask=@n8
   Simple_UseWorldZNormal=StaticBoolParameter() {defaultValue=true, group=Snow, parameterName=UseZNormal}
   Simple_WorldNormal=@n10
   WorldPosition=StaticSwitch(
    True=TransformPosition(None=@n11) {transformSourceType=TRANSFORMPOSSOURCE_Instance, transformType=TRANSFORMPOSSOURCE_World}
    False=WorldPosition().XYZ
    Value=@n12
   )
  ) {materialFunction=/Game/Athena/Environments/Landscape/MaterialFunctions/Snow/MF_Athena_Holiday_SnowFall_S7}
  n14 = FunctionInput(Preview=StaticBool() {value=true}) {bUsePreviewValueAsDefault=true, inputName=Simple Mask, inputType=FunctionInput_StaticBool, previewValue=<1 0 0 1>}
  n15 = StaticSwitch(True=@n13.Simple_SnowMask, False=@n13.Simple_SnowMask, Value=@n14) {defaultValue=true}
  n16 = StaticSwitch(
   True=@n13.SnowMaskNormal
   False=@n15
   Value=FunctionInput(Preview=StaticBool()) {bUsePreviewValueAsDefault=true, inputName=UseSnowMaskNormal, inputType=FunctionInput_StaticBool}
  )
  n17 = SetMaterialAttributes!(
   MaterialAttributes=StaticSwitchParameter(True=BlendMaterialAttributes(A=@n1, B=@n7, Alpha=@n16), False=@n1) {group=Snow, parameterName=UseSnowLegacy}
   WorldPositionOffset=@n9.WorldPositionOffset
  ) {attributeSetTypes [#4]}
  n18 = NamedRerouteUsage()
  n19 = GetMaterialAttributes(None=@n17) {attributeGetTypes [#1, #3]}
  n20 = MaterialFunctionCall!(GrassMask=@n18, RockMask, DesaturationAmount=Constant() {r=1}, UVs, BaseColor=@n19.BaseColor) {materialFunction=/Game/Athena/Environments/Landscape/MaterialFunctions/Snow/MF_Athena_ColorTest_S07}
  n21 = FunctionInput(Preview=StaticBool()) {bUsePreviewValueAsDefault=true, inputName=UseColorTestRock, inputType=FunctionInput_StaticBool, previewValue=<1 0 0 1>}
  n22 = MaterialFunctionCall!(Material=@n17, NewBaseColor=StaticSwitch(True=@n20.ResultWithRock, False=@n20.ResultWithRock, Value=@n21), Mask) {materialFunction=/Engine/Functions/MaterialLayerFunctions/MatLayerBlend_OverrideBaseColor}
  n23 = Constant() {r=0.15}
  n24 = Multiply(A=@n19.SubsurfaceColor, B=@n23)
  FunctionOutput(None=StaticSwitchParameter(
    True=MaterialFunctionCall!(Material=@n22.Blended Material, NewSubSurfaceColor=@n24, NewOpacity, Mask) {materialFunction=/Engine/Functions/MaterialLayerFunctions/MatLayerBlend_OverrideSubSurface}.Blended Material
    False=@n17
   ) {parameterName=UseColorTest}) {outputName=FullyBlendedMA}
  FunctionOutput(None=@n16) {outputName=SnowMask}
  n25 = VectorParameter() {group=Snow, parameterName=SnowColor_1, defaultValue=<rgba 0.7262 0.7933 0.83 1>}
  n26 = Clamp(None=LinearInterpolate(
    A=ScalarParameter() {defaultValue=-0.0381, group=Snow, parameterName=SnowNormalContrast1}
    B=ScalarParameter() {defaultValue=1.4108, group=Snow, parameterName=SnowNormalContrast2}
    Alpha=ComponentMask(None=@n10) {b=true}
   ))
  n27 = Clamp(None=LinearInterpolate(
    A=ScalarParameter() {group=Snow, parameterName=SnowDiffuseContrast1}
    B=ScalarParameter() {defaultValue=13.0863, group=Snow, parameterName=SnowDiffuseContrast2}
    Alpha=ComponentMask(None=@n9.BaseColor) {r=true}
   ))
  n28 = SetMaterialAttributes!(MaterialAttributes=@n9.MaterialAttributes, BaseColor=LinearInterpolate(A=@n9.BaseColor, B=@n25.RGB, Alpha=Multiply(A=@n26, B=@n27))) {attributeSetTypes [#1]}
  n29 = ScalarParameter() {defaultValue=0.2, group=Snow, parameterName=SnowSubsurface}
  n30 = SetMaterialAttributes!(
   MaterialAttributes=@n9.MaterialAttributes
   BaseColor=LinearInterpolate(A=@n9.BaseColor, B=ScalarParameter() {defaultValue=0.5, group=Snow, parameterName=SnowWhite}, Alpha=@n15) {constB=0.6}
   SubsurfaceColor=LinearInterpolate(A=@n9.SubsurfaceColor, B=@n29, Alpha=@n15) {constB=0.5}
  ) {attributeSetTypes [#1, #3]}
  n31 = SetMaterialAttributes!(
   MaterialAttributes=StaticSwitchParameter(True=StaticSwitchParameter(True=@n28, False=@n30) {group=Snow, parameterName=UseSimpleSnow}, False=@n1) {group=Snow, parameterName=UseSnowLegacy}
   WorldPositionOffset=@n9.WorldPositionOffset
  ) {attributeSetTypes [#4]}
  n32 = GetMaterialAttributes(None=@n31) {attributeGetTypes [#1, #3]}
  n33 = MaterialFunctionCall(GrassMask=@n18, BaseColor=@n32.BaseColor) {materialFunction=&MF_Athena_ColorTest_S07}
  n34 = MaterialFunctionCall(Material=@n31, NewBaseColor=StaticSwitch(True=@n33.ResultWithRock, False=@n33.ResultWithRock, Value=@n21)) {materialFunction=&MatLayerBlend_OverrideBaseColor}
  n35 = Multiply(A=@n32.SubsurfaceColor, B=@n23) {constB=0.5}
  FunctionOutput(None=StaticSwitchParameter(
    True=MaterialFunctionCall(Material=@n34.Blended Material, NewSubSurfaceColor=@n35) {materialFunction=&MatLayerBlend_OverrideSubSurface}.Blended Material
    False=@n31
   ) {parameterName=UseColorTest})
  n36 = ScalarParameter() {defaultValue=-0.5, parameterName=SnowUp_Min}
  n37 = ScalarParameter() {defaultValue=1, parameterName=SnowUp_Max}
  n38 = ComponentMask(None=@n10) {b=true}
  LinearInterpolate(A=@n9.BaseColor, B=@n25.RGB, Alpha=Clamp(None=LinearInterpolate(A=@n36, B=@n37, Alpha=@n38)))
  MaterialFunctionCall!(Material=@n2.Material, NewBaseColor=VectorParameter() {parameterName=Snow Tint, defaultValue=<rgba 0.2782 0.3646 0.424 1>}.RGB, Mask) {materialFunction=/Engine/Functions/MaterialLayerFunctions/MatLayerBlend_MultiplyBaseColor}
  Multiply(A=Constant3Vector() {constant=<rgba 0.7917 0.7844 1 1>}, B=@n29)
  NamedRerouteDeclaration(None=@n8) {name=ExtraMask, nodeColor=<rgba 0 1 0.1562 1>}
  ObjectPositionWS()
  ScalarParameter() {defaultValue=0.5, group=Snow, parameterName=SnowWhite}
  ScalarParameter() {defaultValue=0.75, parameterName=Color Burn Desaturation}
  StaticBool()
 ]}
