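# Material function that adds a Season 19 snow layer to tree materials by blending snow base and subsurface colors over the incoming attributes.
# A Fresnel function fed a constant up vector instead of the camera vector builds the mask from a lerp of transformed and vertex normals.
# A static switch optionally applies a bounding-box height gradient.
# Custom primitive data scales the mask by default, with override inputs for it and the Fresnel settings.
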MaterialFunction /Game/Packages/Fortnite_UniqueMaterials_1/MaterialFunctions/Trees/MF_Tree_Snow_s19
graph {nodes [
  n1 = VectorParameter() {group=s19_Snow, parameterName=SnowBaseColor, defaultValue=<rgba 0.7 0.7 0.7 0>}
  n2 = VectorParameter() {group=s19_Snow, parameterName=SSSColor, defaultValue=<rgba 0.0099 0.0588 0.1962 0>}
  n3 = GetMaterialAttributes(None=NamedRerouteDeclaration(None=FunctionInput() {inputName=MA In, inputType=FunctionInput_MaterialAttributes}) {name=MA_In, nodeColor=<rgba 0.6875 0 1 1>}) {attributeGetTypes [#1]}
  FunctionOutput(None=BlendMaterialAttributes(
    A=@n3.MaterialAttributes
    B=SetMaterialAttributes!(MaterialAttributes=@n3.MaterialAttributes, BaseColor=@n1.RGB, SubsurfaceColor=@n2.RGB) {attributeSetTypes [#2, #3]}
    Alpha=NamedRerouteUsage()
   ) {vertexAttributeBlendType=UseA})
  NamedRerouteDeclaration(None=@n3.Normal) {name=Normal_In, nodeColor=<rgba 0.5 0.5 1 1>}
  n4 = FunctionInput(Preview=ScalarParameter() {defaultValue=1, group=s19_Snow, parameterName=SnowBlendBrightness}) {bUsePreviewValueAsDefault=true, inputName=Fresnel_SnowBlendBrightness_Override, inputType=FunctionInput_Scalar, sortPriority=3}
  n5 = FunctionInput(Preview=ScalarParameter() {defaultValue=10, group=s19_Snow, parameterName=SnowBlendExponent}) {bUsePreviewValueAsDefault=true, inputName=Fresnel_SnowBlendExponent_Override, inputType=FunctionInput_Scalar, sortPriority=2}
  n6 = Transform(None=NamedRerouteUsage())
  n7 = VertexNormalWS()
  n8 = ScalarParameter() {defaultValue=0.5, group=s19_Snow, parameterName=NormalLerp}
  n9 = MaterialFunctionCall(
   Brightness=@n4
   Exponent=@n5
   Camera Vector=Constant3Vector() {constant=<rgba 0 0 1 0>}
   Vertex / Pixel Normal=Normalize(VectorInput=LinearInterpolate(A=@n6, B=@n7, Alpha=@n8))
  ) {materialFunction=&MF_Fresnel_FX}
  n10 = Multiply(
   A=@n9.Fresnel Double Sided
   B=MaterialFunctionCall() {materialFunction=&BoundingBoxBased_0-1_UVW}.B
  )
  n11 = ScalarParameter() {bUseCustomPrimitiveData=true, parameterName=s19Snow, primitiveDataIndex=6}
  NamedRerouteDeclaration(None=Multiply(
    A=StaticSwitchParameter(True=@n10, False=@n9.Fresnel Double Sided) {group=s19_Snow, parameterName=UseBoudingBoxGradient}
    B=FunctionInput(Preview=@n11) {bUsePreviewValueAsDefault=true, inputName=SnowMaskOverride, inputType=FunctionInput_Scalar, sortPriority=1}
   )) {name=FinalLerp, nodeColor=<rgba 1 1 1 1>}
 ]}
