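# Material function that adds subsurface shading to tree materials, using a scalar mask to separate leaves from branches.
# It switches incoming material attributes to the Subsurface shading model and derives subsurface color from scaled base color times the mask.
# A camera-distance fade, passed through a vertex interpolator, reduces that color on distant trees.
# The mask also applies a diffuse multiplier to base color and blends between branch and leaf opacity parameters.
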MaterialFunction /Game/Packages/Fortnite_UniqueMaterials_1/MaterialFunctions/Trees/MF_Tree_SubSurface
graph {nodes [
  n1 = GetMaterialAttributes(None=FunctionInput() {inputType=FunctionInput_MaterialAttributes}) {attributeGetTypes [#1]}
  n2 = ScalarParameter() {defaultValue=0.7, parameterName=Subsusrface Diffuse Multiplier}
  n3 = FunctionInput() {inputName=Mask, inputType=FunctionInput_Scalar, sortPriority=3}
  n4 = Multiply(A=@n1.BaseColor, B=ScalarParameter() {defaultValue=1, group=PX_Sub-Surface, parameterName=SSS_Scale})
  n5 = WorldPosition()
  n6 = CameraPositionWS()
  n7 = Divide(
   A=Add(A=Distance(A=@n5.XYZ, B=@n6), B=ScalarParameter() {defaultValue=-10000, group=PX_Sub-Surface, parameterName=SSSFadeDistance_Offset})
   B=ScalarParameter() {defaultValue=12000, group=PX_Sub-Surface, parameterName=SSSFadeDistance}
  )
  n8 = VertexInterpolator(VS=OneMinus(None=Clamp(None=@n7)))
  n9 = SetMaterialAttributes!(
   MaterialAttributes=@n1.MaterialAttributes
   BaseColor=Multiply(A=@n1.BaseColor, B=LinearInterpolate(B=@n2, Alpha=@n3) {constA=1, constB=0.7}) {constB=0.7}
   SubsurfaceColor=Multiply(A=Multiply(A=@n4, B=@n3), B=ComponentMask(None=@n8.PS) {r=true})
   Opacity=LinearInterpolate(
    A=ScalarParameter() {defaultValue=1, group=PX_Sub-Surface Opacity, parameterName=Opacity branches}
    B=ScalarParameter() {defaultValue=1, group=PX_Sub-Surface Opacity, parameterName=Opacity leaves}
    Alpha=@n3
   )
   ShadingModel=ShadingModel() {shadingModel=MSM_Subsurface}
  ) {attributeSetTypes [
    #1
    #2
    #3
    #4
   ]}
  FunctionOutput(None=@n9)
 ]}
